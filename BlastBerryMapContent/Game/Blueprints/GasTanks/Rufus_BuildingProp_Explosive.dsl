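; Explosive gas tank/pump building prop that damages players, vehicles and vehicle passengers when it explodes.
; Damage applies the explode GameplayEffect through the last instigator's ability system, falling back to global environment damage without one.
; A DamagedPlayers array guards against double-hitting players, including passengers.
; Hit-actor handling adds stopgap cases: Bush wearers get damaged and can lose the bush, and Biplanes are hit even behind walls.
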(event EventBeginPlay
  (|CallParentFunction) ; This node is disabled and will not be called.\nDrag off pins to build functionality.
)

(event Collision|EventActorBeginOverlap
  (params (OtherActor :type Actor%20Object%20Reference)))

(event EventTick
  (params (DeltaSeconds :type Float%20%28single-precision%29)))

(event EventDestroyHitActors
  (params (HitActor :type Actor%20Object%20Reference))
  (|CallParentFunction HitActor)
  ;[1 1 1] Not good but doing to not have to mess with even more assets for 27.10 there are only a handful of tanks/pumps so going this route. Otherwise removing the WeaponCapsule Trace from Athena_Player_Bush and moving the bush removal into the ApplyDamageToActor function in the gas pump could be a better alternative but would need to test that other interactions don't break
  (CallFunction|HandlePlayerswithBushItemEquipped HitActor)
  ;[1 1 1] If the Biplane than do damage as the line trace hits underneath the vehicle so it is always False. This will always hit the Biplane even if it is behind a wall so again not a good solution but the least disruptive one for 27.10.
  (bind _asfort_vehicle_interface (Utilities|Casting|CastToFortVehicleInterface HitActor))
  (if (GameplayTags|HasTag (Class|FortVehicleInterface|GetVehicleTags _asfort_vehicle_interface) "(TagName=\"Vehicle.Biplane\")" true)
    (CallFunction|ExecuteExplosiononActor HitActor)))

(fn ApplyDamageToActor
  (params (Actor%20to%20Damage :type Actor%20Object%20Reference))
  (bind _actor_to_damage (Variables|GetActor%20to%20Damage))
  (bind _actor_to_damage_2 (Variables|GetActor%20to%20Damage))
  (Utilities|IsValid (Variables|Default|GetLastDamageInstigator)
    (:Is%20Valid
      ;[1 1 1] If the Actor to Damage is a Player Pawn Athena only damage if the player has not already been damaged by this exploding prop, so we don't Double apply damage to the player.
      (bind _asfort_player_pawn (Utilities|Casting|CastToFortPlayerPawn _actor_to_damage)
        (:then
          (if (not (Utilities|Array|ContainsItem (Variables|Default|GetDamagedPlayers) _asfort_player_pawn))
            (Utilities|Array|AddUnique (Variables|Default|GetDamagedPlayers) _asfort_player_pawn)))
        (:CastFailed
          _))
      (GameplayEffects|FortApplyGameplayEffectToTarget :self (Ability|GetFortAbilitySystemComponent (Pawn|GetControlledPawn (Variables|Default|GetLastDamageInstigator))) :GameplayEffectClass (Variables|GameplayEffects|GetGE_Explode) :TargetOfEffect (Ability|GetFortAbilitySystemComponent _actor_to_damage) :OptionalSourceObject self :OptionalEffectCauser self :OptionalDamageSource self :Level 0.0)
      (CallFunction|DealDamagetoVehiclePassengers (Variables|GetActor%20to%20Damage)) ; Handle dealing damage to Vehicle Driver
    )
    (:Is%20Not%20Valid
      (bind _asfort_player_pawn_2 (Utilities|Casting|CastToFortPlayerPawn (Variables|GetActor%20to%20Damage))
        (:then
          ;[1 1 1] Only apply the Explode GE on a player that has not already been damaged
          (if (not (Utilities|Array|ContainsItem (Variables|Default|GetDamagedPlayers) _asfort_player_pawn_2))
            (Utilities|Array|AddUnique (Variables|Default|GetDamagedPlayers) _asfort_player_pawn_2)))
        (:CastFailed
          (Utilities|Casting|CastToFortVehicleInterface _actor_to_damage_2
            (:then
              (Environment|ApplyGlobalEnvironmentDamagetoActor (Variables|GetActor%20to%20Damage) (ScalableFloat|GetValueatLevel (Variables|Default|GetVehicleDamage)))
              (CallFunction|DealDamagetoVehiclePassengers _actor_to_damage_2) ; Handle dealing damage to vehicle driver
            )
            (:CastFailed
              _))))
      (Environment|ApplyGlobalEnvironmentGameplayEffecttoActor (Variables|GetActor%20to%20Damage) (Variables|GameplayEffects|GetGE_Explode)))))

(fn DealDamageToVehiclePassengers
  (params (ActorDamaged :type Actor%20Object%20Reference))
  (bind _lastdamageinstigator (Variables|Default|GetLastDamageInstigator))
  (bind _returnvalue (Pawn|GetControlledPawn _lastdamageinstigator))
  (bind _returnvalue_2 (Ability|GetFortAbilitySystemComponent _returnvalue))
  ;[1 1 1] Check if the Actor Damaged is a Vehcile and has 1 or more passengers aboard.
  (bind _asfort_vehicle_interface (Utilities|Casting|CastToFortVehicleInterface ActorDamaged))
  (if (> (Utilities|Array|Length (Vehicle|GetAllPassengers _asfort_vehicle_interface)) 0)
    (LocalVariables|SetAllPassengers (Vehicle|GetAllPassengers _asfort_vehicle_interface))
    (Utilities|IsValid (Variables|Default|GetLastDamageInstigator)
      (:Is%20Valid
        ;[1 1 1] Apply Damage to all passengers in the vehicle if they have not already been damaged. *NOTE this will not damage teammates or friendly vehicles which is by design
        (for _array_element (LocalVariables|GetAllPassengers)
          (Utilities|IsValid _array_element
            (:Is%20Valid
              (if (not (Utilities|Array|ContainsItem (Variables|Default|GetDamagedPlayers) _array_element))
                (Utilities|IsValid (Ability|GetFortAbilitySystemComponent _array_element)
                  (:Is%20Valid
                    (Utilities|Array|AddUnique (Variables|Default|GetDamagedPlayers) _array_element)
                    (Utilities|IsValid _returnvalue_2
                      (:Is%20Valid
                        (GameplayEffects|FortApplyGameplayEffectToTarget :self _returnvalue_2 :GameplayEffectClass (Variables|GameplayEffects|GetGE_Explode) :TargetOfEffect (Ability|GetFortAbilitySystemComponent _array_element) :OptionalSourceObject self :OptionalEffectCauser self :OptionalDamageSource self :Level 0.0))
                      (:Is%20Not%20Valid
                        (Environment|ApplyGlobalEnvironmentGameplayEffecttoActor _array_element (Variables|GameplayEffects|GetGE_Explode)) ; Edge Case: If Last Damage Instigator was eliminated before applying the Explode GE to this player., Last Damage Instigator becomes invalid, so we apply the GE without the Instigator
                      )))))))))
      (:Is%20Not%20Valid
        ;[1 1 1] Apply Explode GE when there is no damage instigator available and the Player has not already been damaged
        (for _array_element_2 (LocalVariables|GetAllPassengers)
          (Utilities|IsValid _array_element_2
            (:Is%20Valid
              (if (not (Utilities|Array|ContainsItem (Variables|Default|GetDamagedPlayers) _array_element_2))
                (Utilities|Array|AddUnique (Variables|Default|GetDamagedPlayers) _array_element_2)
                (Environment|ApplyGlobalEnvironmentGameplayEffecttoActor _array_element_2 (Variables|GameplayEffects|GetGE_Explode)))))))) ; Check if there is a valid Damage Instigator
  ))

(fn HandlePlayersWithBushItemEquipped
  (params (ActorRef :type Actor%20Object%20Reference))
  (bind _lastdamageinstigator (Variables|Default|GetLastDamageInstigator))
  ;[1 1 1] If Actor to Damage is a player pawn and has the Bush equipped apply damage if not done yet (Damage won't be applied if a teammate instigated the explosion) and remove the bush from the player if not instigated by a teammate.
  (bind _asfort_player_pawn (Utilities|Casting|CastToFortPlayerPawn ActorRef))
  (if (GameplayTags|HasMatchingGameplayTag _asfort_player_pawn "(TagName=\"Athena.Item.Bush\")")
    (if (not (Utilities|Array|ContainsItem (Variables|Default|GetDamagedPlayers) _asfort_player_pawn))
      (CallFunction|ExecuteExplosiononActor ActorRef) ; Only Execute Explosion for Actor to Damage if the player hasn't been damaged by this prop yet.
    )
    (Utilities|IsValid _lastdamageinstigator
      (:Is%20Valid
        (if (and (Team|OnSameTeam (Game|Player|GetPlayerPawn _lastdamageinstigator) ActorRef) (!= (Game|Player|GetPlayerPawn _lastdamageinstigator) ActorRef))
          _
          (else
            _)))
      (:Is%20Not%20Valid
        _))
    (Ability|SendGameplayEventtoActor ActorRef "(TagName=\"Abilities.Generic.Athena.Bush.Destroyed\")")))

(fn ConstructionScript
  (|CallParentFunction))
